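# Unlit translucent material that procedurally builds a color lookup gradient from UVs, located among the foliage Blueprint tools.
# V splits into four bands via saturate, ceil and floor masks; the first three take Color1 to Color3 parameter sets, the last stays black.
# U is quantized by Steps with SoftFloor-softened edges, then blends each band through left, middle and right colors.
# The same U blend drives a variation brightness multiplier on the combined color.
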Material /Game/Environments/Apollo/Foliage/BlueprintTools/M_ColorLut
graph {
 properties {blendMode=BLEND_Translucent, pixelDepthOffsetMode=PDOM_Legacy, shadingModel=MSM_Unlit, usedShadingModels=Unlit}
 nodes [
  n1 = Multiply(A=ComponentMask(None=TextureCoordinate()) {g=true}) {constB=4}
  n2 = Saturate(None=@n1)
  n3 = Multiply(A=Ceil(None=@n2), B=OneMinus(None=Floor(None=@n2)))
  n4 = Multiply(A=@n3, B=ScalarParameter() {defaultValue=1, parameterName=Birghtness_Light})
  n5 = Saturate(None=Add(A=@n1) {constB=-1})
  n6 = Multiply(A=Ceil(None=@n5), B=OneMinus(None=Floor(None=@n5)))
  n7 = Multiply(A=@n6, B=ScalarParameter() {defaultValue=0.66, parameterName=Birghtness_Neutral})
  n8 = Saturate(None=Add(A=@n1) {constB=-2})
  n9 = Multiply(A=Ceil(None=@n8), B=OneMinus(None=Floor(None=@n8)))
  n10 = ScalarParameter() {defaultValue=0.3, parameterName=Birghtness_Shadows}
  n11 = Saturate(None=Add(A=@n1) {constB=-3})
  n12 = Ceil(None=@n11)
  n13 = OneMinus(None=Floor(None=@n11))
  Add(A=Add(A=Add(A=@n4, B=@n7), B=Multiply(A=@n9, B=@n10)), B=Multiply(A=Multiply(A=@n12, B=@n13)) {constB=0})
  n14 = ScalarParameter() {defaultValue=32, parameterName=Steps}
  n15 = Multiply(A=ComponentMask(None=TextureCoordinate()) {r=true}, B=@n14) {constB=16}
  n16 = Floor(None=@n15)
  n17 = Frac(None=@n15)
  n18 = Multiply(A=@n17, B=OneMinus(None=@n17))
  n19 = Saturate(None=Power(Base=Multiply(A=@n18) {constB=2}, Exp=ScalarParameter() {defaultValue=0.1, parameterName=SoftFloor}))
  n20 = Multiply(A=Divide(A=LinearInterpolate(A=@n15, B=@n16, Alpha=@n19), B=@n14) {constB=16}) {constB=2}
  n21 = Saturate(None=@n20)
  n22 = OneMinus(None=@n21)
  n23 = Saturate(None=Add(A=@n20) {constB=-1})
  Add(A=@n22, B=Multiply(A=@n23) {constB=-1})
  Constant() {r=1}
  n24 = VectorParameter() {parameterName=Color1_left, defaultValue=<rgba 1 0 0 1>}
  n25 = VectorParameter() {parameterName=Color1, defaultValue=<rgba 1 0 0 1>}
  n26 = Multiply(
   A=LinearInterpolate(
    A=LinearInterpolate(A=@n24.RGB, B=@n25.RGB, Alpha=@n21)
    B=VectorParameter() {parameterName=Color1_right, defaultValue=<rgba 1 0 0 1>}.RGB
    Alpha=@n23
   )
   B=@n3
  )
  n27 = VectorParameter() {parameterName=Color2_left, defaultValue=<rgba 0 1 0 1>}
  n28 = VectorParameter() {parameterName=Color2, defaultValue=<rgba 0 1 0 1>}
  n29 = Multiply(
   A=@n6
   B=LinearInterpolate(
    A=LinearInterpolate(A=@n27.RGB, B=@n28.RGB, Alpha=@n21)
    B=VectorParameter() {parameterName=Color2_right, defaultValue=<rgba 0 1 0 1>}.RGB
    Alpha=@n23
   )
  )
  n30 = LinearInterpolate(
   A=LinearInterpolate(
    A=VectorParameter() {parameterName=Color3_left, defaultValue=<rgba 0 0 1 1>}.RGB
    B=VectorParameter() {parameterName=Color3, defaultValue=<rgba 0 0 1 1>}.RGB
    Alpha=@n21
   )
   B=VectorParameter() {parameterName=Color3_right, defaultValue=<rgba 0 0 1 1>}.RGB
   Alpha=@n23
  )
  n31 = ScalarParameter() {defaultValue=0.9, parameterName=VariationBirghtness_Neutral}
  n32 = ScalarParameter() {defaultValue=0.3, parameterName=VariationBirghtness_Dark}
  Multiply(
   A=Add(A=Add(A=@n26, B=@n29), B=Multiply(A=@n9, B=@n30))
   B=LinearInterpolate(A=LinearInterpolate(A=@n31, B=@n32, Alpha=@n22), B=ScalarParameter() {defaultValue=1, parameterName=VariationBirghtness_Light}, Alpha=@n23)
  )
  Power(Base=@n20)
 ]
}
