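; Editor tool Blueprint for authoring tree foliage color LUT textures from nine colors: light, neutral and shadow, each with center, left and right variations.
; Construction creates a dynamic M_ColorLut instance, feeds it the colors and draws it into a preview render target.
; Updatetexture converts that render target into the chosen texture asset via editor scripting and shows it on the static mesh.
; GetColorFromAsset reverses this, sampling fixed UVs of an existing texture to refill the colors.
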(event EventBeginPlay)

(event Collision|EventActorBeginOverlap
  (params (OtherActor :type Actor%20Object%20Reference)))

(event Custom|UpdateRenderTarget
  (CallFunction|UpdateMaterial)
  (Rendering|DrawMaterialtoRenderTarget 0 (Variables|Tool|GetRenderTargetPreview) (Variables|Default|GetMID)))

(event EventTick
  (params (DeltaSeconds :type Float%20%28single-precision%29)))

(event Custom|Updatetexture
  (bind _textureasset (Variables|Tool|GetTextureAsset))
  (CallFunction|UpdateRenderTarget)
  (EditorScripting|Rendering|ConvertRenderTargettoTexture2DEditorOnly 0 (Variables|Tool|GetRenderTargetPreview) _textureasset)
  (bind _returnvalue (Rendering|Material|CreateDynamicMaterialInstance (Variables|Default|GetStaticMesh) 0 "/Game/Environments/Apollo/Foliage/BlueprintTools/Test_LUT.Test_LUT" "None" :!class /Script/Engine.PrimitiveComponent))
  (Rendering|Material|SetTextureParameterValue _returnvalue "Texture" _textureasset))

(event Custom|GetColorFromAsset
  (Variables|1_Color_Center_Base|SetColor_Neutral (Rendering|Texture2DSampleUVEditorOnly (Variables|Tool|GetTextureAsset) "(X=0.500000,Y=0.375000)"))
  (Variables|1_Color_Center_Base|SetColor_Light (Rendering|Texture2DSampleUVEditorOnly (Variables|Tool|GetTextureAsset) "(X=0.500000,Y=0.125000)"))
  (Variables|1_Color_Center_Base|SetColor_Shadow (Rendering|Texture2DSampleUVEditorOnly (Variables|Tool|GetTextureAsset) "(X=0.500000,Y=0.625000)"))
  (Variables|2ColorVariationLeft|SetColor_Var_Neutral_Left (Rendering|Texture2DSampleUVEditorOnly (Variables|Tool|GetTextureAsset) "(X=0.000100,Y=0.375000)"))
  (Variables|2ColorVariationLeft|SetColor_Var_Light_Left (Rendering|Texture2DSampleUVEditorOnly (Variables|Tool|GetTextureAsset) "(X=0.000100,Y=0.125000)"))
  (Variables|2ColorVariationLeft|SetColor_Var_Shadows_Left (Rendering|Texture2DSampleUVEditorOnly (Variables|Tool|GetTextureAsset) "(X=0.000100,Y=0.625000)"))
  (Variables|3_ColorVariationRight|SetColor_Var_Neutral_Right (Rendering|Texture2DSampleUVEditorOnly (Variables|Tool|GetTextureAsset) "(X=0.999990,Y=0.375000)"))
  (Variables|3_ColorVariationRight|SetColor_Var_Light_Right (Rendering|Texture2DSampleUVEditorOnly (Variables|Tool|GetTextureAsset) "(X=0.999990,Y=0.125000)"))
  (Variables|3_ColorVariationRight|SetColor_Var_Shadows_Right (Rendering|Texture2DSampleUVEditorOnly (Variables|Tool|GetTextureAsset) "(X=0.999990,Y=0.625000)"))
  (CallFunction|UpdateMaterial)
  (CallFunction|UpdateRenderTarget))

(fn UpdateMaterial
  (bind _mid (Variables|Default|GetMID))
  (Rendering|Material|SetVectorParameterValue _mid "Color1" (Variables|1_Color_Center_Base|GetColor_Light) :!class /Script/Engine.MaterialInstanceDynamic)
  (Rendering|Material|SetVectorParameterValue _mid "Color2" (Variables|1_Color_Center_Base|GetColor_Neutral) :!class /Script/Engine.MaterialInstanceDynamic)
  (Rendering|Material|SetVectorParameterValue _mid "Color3" (Variables|1_Color_Center_Base|GetColor_Shadow) :!class /Script/Engine.MaterialInstanceDynamic)
  (Rendering|Material|SetVectorParameterValue _mid "Color1_left" (Variables|2ColorVariationLeft|GetColor_Var_Light_Left) :!class /Script/Engine.MaterialInstanceDynamic)
  (Rendering|Material|SetVectorParameterValue _mid "Color2_left" (Variables|2ColorVariationLeft|GetColor_Var_Neutral_Left) :!class /Script/Engine.MaterialInstanceDynamic)
  (Rendering|Material|SetVectorParameterValue _mid "Color3_left" (Variables|2ColorVariationLeft|GetColor_Var_Shadows_Left) :!class /Script/Engine.MaterialInstanceDynamic)
  (Rendering|Material|SetVectorParameterValue _mid "Color1_right" (Variables|3_ColorVariationRight|GetColor_Var_Light_Right) :!class /Script/Engine.MaterialInstanceDynamic)
  (Rendering|Material|SetVectorParameterValue _mid "Color2_right" (Variables|3_ColorVariationRight|GetColor_Var_Neutral_Right) :!class /Script/Engine.MaterialInstanceDynamic)
  (Rendering|Material|SetVectorParameterValue _mid "Color3_right" (Variables|3_ColorVariationRight|GetColor_Var_Shadows_Right) :!class /Script/Engine.MaterialInstanceDynamic))

(fn ConstructionScript
  (Utilities|FlowControl|Sequence
    (:then_0
      (bind _returnvalue (Rendering|Material|CreateDynamicMaterialInstance 0 "/Game/Environments/Apollo/Foliage/BlueprintTools/M_ColorLut.M_ColorLut" "None" "None" :!class /Script/Engine.KismetMaterialLibrary))
      (Variables|Default|SetMID _returnvalue)
      (CallFunction|UpdateMaterial))
    (:then_1
      (Variables|Tool|SetRenderTargetPreview "/Game/Environments/Apollo/Foliage/BlueprintTools/NewTextureRenderTarget2D.NewTextureRenderTarget2D"))
    (:then_2
      (CallFunction|UpdateRenderTarget))))
